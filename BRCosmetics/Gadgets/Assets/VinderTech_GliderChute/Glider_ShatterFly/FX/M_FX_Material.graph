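# Translucent unlit FX material for the ShatterFly glider, with tinted gradient color, cubemap reflection and refraction.
# Panning cellular noise perturbs the normal map, warps a second noise and varies the IOR, which is flat on mobile and lower quality.
# Normalized distance from object origin, noise and a squared facing term index a curve-atlas color, scaled by time of day.
# The facing term also shapes opacity, multiplied by FXMask and BP_Opacity.
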Material /BRCosmetics/Gadgets/Assets/VinderTech_GliderChute/Glider_ShatterFly/FX/M_FX_Material
graph {
 properties {
  bEnableResponsiveAA=true
  bUsedWithMeshDeformer=true
  bUsedWithSkeletalMesh=true
  blendMode=BLEND_Translucent
  pixelDepthOffsetMode=PDOM_Legacy
  refractionMethod=RM_IndexOfRefraction
  shadingModel=MSM_Unlit
  usedShadingModels=Unlit
 }
 nodes [
  n1 = ScalarParameter() {defaultValue=0.2, parameterName=Ref}
  n2 = TextureCoordinate()
  n3 = ComponentMask(None=VectorParameter() {parameterName=UV, defaultValue=<rgba 2 0.75 6 0.5>}.RGB) {g=true, r=true}
  n4 = TextureSample(UVs=Panner(Coordinate=Multiply(A=@n2, B=@n3)) {speedX=0.03, speedY=0.1}) {samplerType=SAMPLERTYPE_Masks, texture=/Game/Global/Textures/Default/Noise/cellularNoise}
  n5 = LinearInterpolate(A=Subtract(B=@n1), B=Add(B=@n1) {constA=1}, Alpha=ComponentMask(None=@n4.RGB) {b=true}) {constA=0.94, constB=1.06}
  LinearInterpolate(B=@n5) {constA=1, constAlpha=0.2}
  n6 = Constant() {r=1}
  n7 = Multiply(
   A=ScalarParameter() {defaultValue=1, parameterName=BP_Opacity}
   B=TextureSampleParameter2D() {parameterName=FXMask, texture=/Engine/EngineMaterials/DefaultWhiteGrid_Low}.G
  )
  LinearInterpolate(
   A=@n6
   B=QualitySwitch(Default=ShadingPathSwitch(Default=@n5, Mobile=@n6) {desc=Migrated from FeatureLevelSwitch}, Low=@n6, High=@n6, Medium=@n6)
   Alpha=Multiply(A=ScalarParameter() {defaultValue=1, parameterName=SepTrans}, B=@n7)
  )
  n8 = ObjectPositionWS()
  n9 = WorldPosition()
  n10 = Divide(
   A=MaterialFunctionCall(Vector 3=Subtract(A=@n8, B=@n9.XYZ)) {materialFunction=/Engine/Functions/Engine_MaterialFunctions02/Utility/VectorLength}.V3 Length
   B=Multiply(A=ScalarParameter() {defaultValue=0.487, parameterName=Range}, B=ObjectRadius())
  )
  n11 = TextureCoordinate()
  n12 = VectorParameter() {parameterName=UV, defaultValue=<rgba 2 0.75 6 0.5>}
  n13 = AppendVector(A=@n12.B, B=@n12.B)
  n14 = Add(
   A=Panner(Coordinate=Multiply(A=@n11, B=@n13)) {speedX=0.1, speedY=0.2}
   B=Multiply(A=ComponentMask(None=@n4.RGB) {g=true, r=true}, B=ScalarParameter() {defaultValue=0.5, parameterName=TextureDisplace}) {constB=0.1}
  )
  n15 = Multiply(
   A=Subtract(A=TextureSample(UVs=@n14) {samplerType=SAMPLERTYPE_LinearColor, texture=&T_NoisePack_LinearColor}.R) {constB=0.5}
   B=ScalarParameter() {defaultValue=-0.25, parameterName=MaskOffset2}
  )
  n16 = TextureSampleParameter2D() {parameterName=N, samplerType=SAMPLERTYPE_Normal, texture=/Engine/EngineMaterials/BaseFlattenNormalMap}
  n17 = MaterialFunctionCall(BaseNormal=@n16.RGB, AdditionalNormal=@n4.RGB) {materialFunction=&BlendAngleCorrectedNormals}
  n18 = VertexNormalWS()
  n19 = CameraVectorWS()
  n20 = ScalarParameter() {parameterName=UseCamVec0-1}
  n21 = Abs(None=DotProduct(A=Transform(None=@n17.Result), B=LinearInterpolate(A=@n18, B=@n19, Alpha=@n20)))
  n22 = Abs(None=ComponentMask(None=VertexNormalWS()) {b=true})
  n23 = Multiply(A=LinearInterpolate(Alpha=OneMinus(None=@n21)) {constA=0.75}, B=LinearInterpolate(Alpha=OneMinus(None=@n22)) {constA=0.5})
  n24 = Multiply(A=@n23, B=@n23)
  n25 = ScalarParameter() {defaultValue=-0.5, parameterName=MaskOffset}
  n26 = CurveAtlasRowParameter(CurveTime=Add(A=Add(A=@n10, B=@n15), B=Multiply(A=@n24, B=@n25) {constB=0.25})) {
   atlas=&AthenaCurveAtlas01
   curve=/Game/Gadgets/Assets/VinderTech_GliderChute/Glider_ShatterFly/FX/Curve_Color
   defaultValue=74
  }
  n27 = Constant() {r=0.8}
  n28 = Multiply(
   A=Multiply(A=@n26, B=@n24)
   B=MaterialFunctionCall(Color_Morning=@n27, Color_Day=Constant() {r=1}, Color_Evening=@n27, Color_Night=Constant() {r=0.6}) {materialFunction=&Color_Time_of_Day}.Result
  )
  n29 = TextureSample(UVs=ReflectionVectorWS()) {samplerType=SAMPLERTYPE_LinearColor, texture=/Game/Weapons/WeaponSkins/Textures/AthenaCubemap2}
  n30 = ShadingPathSwitch(
   Default=Multiply(
    A=MaterialFunctionCall() {materialFunction=&Color_Time_of_Day}.Result
    B=Multiply(A=@n29.RGB) {constB=1.25}
   )
   Mobile=Constant()
  ) {desc=Migrated from FeatureLevelSwitch}
  MaterialFunctionCall(Emissive=Multiply(A=Add(A=@n28, B=@n30), B=VectorParameter() {parameterName=ColorTint, defaultValue=<rgba 1 1 1 0>}.RGB)) {materialFunction=&MF_FX_EmissiveAdjustments}
  Reroute(None=@n16.RGB)
  n31 = ScalarParameter() {defaultValue=0.33, parameterName=OpacityExp}
  Reroute(None=Multiply(A=Power(Base=@n24, Exp=@n31), B=@n7))
 ]
}
